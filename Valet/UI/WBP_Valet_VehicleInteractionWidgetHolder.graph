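# Widget Blueprint that serves as the holder for vehicle interaction widgets in the Valet UI, built on the native FortVehicleInteractionWidgetHolder class.
# Its only widget is the inherited root VerticalBox_WidgetHolder.
# When interaction context info updates, the event graph passes it straight to HandleContextInfoUpdated on itself.
# The Blueprint declares no variables and adds no other logic.
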WidgetBlueprint /Valet/UI/WBP_Valet_VehicleInteractionWidgetHolder
graphs [EventGraph {nodes [
   n1 = Event OnContextInfoUpdated!() -> (OutputDelegate, then, ContextInfo:Fort Interact Context Info Object Reference) {is=title K2Node_Event}
   > HandleContextInfoUpdated!(
    execute
    self:Fort Vehicle Interaction Widget Holder Object Reference
    ContextInfo:Fort Interact Context Info Object Reference=@n1.ContextInfo
   ) -> (then)
  ]}]
variables []
widgets {
 info {
  inheritedWidgetCount=1
  parentClass=/Script/ValetUI.FortVehicleInteractionWidgetHolder
  rootWidgetClass=&VerticalBox
  widgetCount=1
 }
 widgets [VerticalBox_WidgetHolder {widgetClassPath=&VerticalBox, bInherited=true, widget=VerticalBox_WidgetHolder}]
}
